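; Parent Blueprint for NeonBurn environmental props that always die when asked and burst into a healing splash.
; On server death, a scale-adjusted sphere overlap applies Curie water to hit actors and the NeonBurn healing effect to pawns, credited to the instigator.
; QuestHandling then sends an extinguish quest stat if something was put out.
; Death effects spawn a Niagara burst and a randomly rotated, scaled, fading decal.
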(event EventOnDeathServer
  (params (Damage :type Float%20%28single-precision%29) (DamageTags :type Gameplay%20Tag%20Container%20Structure%20%28by%20ref%29) (Momentum :type Vector :default "0, 0, 0") (HitInfo :type Hit%20Result%20Structure%20%28by%20ref%29) (InstigatedBy :type Controller%20Object%20Reference) (DamageCauser :type Actor%20Object%20Reference) (EffectContext :type Gameplay%20Effect%20Context%20Handle%20Structure))
  ;[1 0.4431 0] EventOnDeathServer
  ; Explode and grant NeonBurn effect to pawns in the radius, attempt to spawn a NeonBurn related pick up (if applicable)
  (if (Class|AthenaFunctionLibrary|ShouldAllowDeathEffects DamageTags)
    (CallFunction|DeactivateEmitter)
    (Variables|Explosion|Curie|Water|SetWaterInstigator InstigatedBy)
    (bind (_outactors _returnvalue) (Collision|SphereOverlapActors (Transformation|GetActorLocation) (* (Variables|Explosion|GetExplosion%20Radius) (Math|Vector|VectorLength (Transformation|GetActorScale3D))) (Variables|Explosion|GetExplosionHitObjectTypes)))
    (if _returnvalue
      (for _array_element _outactors
        ;;[0.272 0.5899 1] For Each Pawn or environment actor hit in the Splash radius, provide healing and
        (Utilities|IsValid _array_element
          (:Is%20Valid
            (bind _output_get (Variables|Explosion|SetExplosionHitActor _array_element))
            ;;;[0 0.06265 1] Apply water to hit actors.
            ;;; Depricated(?) system that puts out dynamic fire (Curie)
            (CallFunction|ApplyWaterIfCurieEnabled (Variables|Explosion|Curie|Water|GetWaterInstigator) _output_get)
            ;;;[0 0.06265 1] Attempt to heal any pawn hit by the splash effect
            (bind _asfort_pawn (Utilities|Casting|CastToFortPawn (Variables|Explosion|GetExplosionHitActor)))
            (GameplayEffects|FortApplyGameplayEffectToTarget (Ability|GetFortAbilitySystemComponent self) (Variables|NeonBurn|GetGE_NeonBurnHealing) (Ability|GetAbilitySystemComponent _asfort_pawn) self self 0 (Variables|NeonBurn|GetNeonBurnHealingLevel) (select (Utilities|IsValid (Ability|GetAbilitySystemComponent (Variables|Explosion|Curie|Water|GetWaterInstigator)) :!class /Script/Engine.KismetSystemLibrary) (Class|AthenaFunctionLibrary|MakeSimpleFortEffectContext (Variables|Explosion|Curie|Water|GetWaterInstigator)) 0)))))
      (CallFunction|QuestHandling))))

(event EventOnDeathPlayEffects
  (params (Damage :type Float%20%28single-precision%29) (DamageTags :type Gameplay%20Tag%20Container%20Structure%20%28by%20ref%29) (Momentum :type Vector :default "0, 0, 0") (HitInfo :type Hit%20Result%20Structure%20%28by%20ref%29) (InstigatedBy :type Fort%20Pawn%20Object%20Reference) (DamageCauser :type Actor%20Object%20Reference) (EffectContext :type Gameplay%20Effect%20Context%20Handle%20Structure))
  ;[0.26 0 0.5495] OnDeathPlayEffects
  ; VFX burst effect and decal
  (|CallParentFunction Damage DamageTags Momentum HitInfo InstigatedBy DamageCauser EffectContext)
  (bind _returnvalue (Niagara|SpawnSystematLocation :SystemTemplate (Variables|FX|GetFX_DestroyedBurst) :Location (Transformation|GetActorLocation) :Scale "1.000000,1.000000,1.000000" :bAutoDestroy true :bAutoActivate true :PoolingMethod "AutoRelease" :bPreCullCheck true))
  (Niagara|SetNiagaraVariable(Vector3) _returnvalue "Offset" (Variables|FX|GetFX%20Offset%20Spawn))
  (bind _returnvalue_2 (Rendering|Decal|SpawnDecalatLocation (Variables|FX|Decal|GetDeath%20Decal%20Material) (Math|Vector|MakeVector (* (.x (Variables|FX|Decal|GetDeath%20Decal%20Size)) (.x (Transformation|GetActorScale3D))) (* (.y (Variables|FX|Decal|GetDeath%20Decal%20Size)) (.y (Transformation|GetActorScale3D))) (* (.z (Variables|FX|Decal|GetDeath%20Decal%20Size)) (.z (Transformation|GetActorScale3D)))) (Transformation|GetActorLocation) (Math|Rotator|MakeRotator (Math|Random|RandomFloatinRangefromStream (Variables|FX|Decal|GetRandom%20Stream) 0.0 359.0) -9e+01)))
  (Rendering|Components|Decal|SetFadeIn _returnvalue_2 0.0 0.25)
  (Rendering|Components|Decal|SetFadeOut _returnvalue_2 0.75 1.75 true))

(fn DistanceCheckforCameraLensEffect
  (params (Instigator :type Actor%20Object%20Reference))
  (bind (_returnvalue_x _returnvalue_y _returnvalue_z) (Math|Vector|BreakVector (Transformation|GetActorLocation)))
  (bind (_returnvalue_x_2 _returnvalue_y_2 _returnvalue_z_2) (Math|Vector|BreakVector (Transformation|GetActorLocation Instigator)))
  (return (Math|Boolean|ORBoolean (or (> _returnvalue_x_2 (+ _returnvalue_x (Variables|FX|LensEffect|GetDistance%20for%20Camera%20Lens))) (< _returnvalue_x_2 (- _returnvalue_x (Variables|FX|LensEffect|GetDistance%20for%20Camera%20Lens)))) (or (> (+ _returnvalue_y (Variables|FX|LensEffect|GetDistance%20for%20Camera%20Lens)) _returnvalue_y_2) (< (- _returnvalue_y (Variables|FX|LensEffect|GetDistance%20for%20Camera%20Lens)) _returnvalue_y_2)) (or (> _returnvalue_z_2 (+ _returnvalue_z (Variables|FX|LensEffect|GetDistance%20for%20Camera%20Lens))) (< _returnvalue_z_2 (- _returnvalue_z (Variables|FX|LensEffect|GetDistance%20for%20Camera%20Lens)))))))

(fn QuestHandling
  (bind _waterinstigator (Variables|Explosion|Curie|Water|GetWaterInstigator))
  (bind _returnvalue (Fortnite|GetControllerfromActor _waterinstigator))
  (if (Variables|Explosion|Curie|Water|GetSomethingWasPutOut)
    (Utilities|IsValid _returnvalue
      (:Is%20Valid
        (bind _asfort_player_controller (Utilities|Casting|CastToFortPlayerController _returnvalue))
        (Quest|SendComplexCustomStatEvent (Quests|GetQuestManager _asfort_player_controller "Invalid") 0 "(GameplayTags=((TagName=\"Athena.Quest.Extinguish\")))" "(GameplayTags=)" (Variables|Explosion|Curie|Water|GetCountOfItemsExtinguished))
        (Variables|Explosion|Curie|Water|SetSomethingWasPutOut false)))))

(fn ShouldDie
  (params (Damage :type Float%20%28single-precision%29) (EventInstigator :type Controller%20Object%20Reference) (DamageCauser :type Actor%20Object%20Reference) (bIsTest :type Boolean) (EffectContext :type Gameplay%20Effect%20Context%20Handle%20Structure%20%28by%20ref%29))
  (return true))

(fn ConstructionScript
  (|CallParentFunction))
