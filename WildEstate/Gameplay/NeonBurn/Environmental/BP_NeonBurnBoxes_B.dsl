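; Destructible environmental box prop for NeonBurn that removes its instanced static mesh collision when destroyed.
; On server death it calls the parent handler, then sets the DisableInstancedSMCollision RepNotify flag.
; The OnRep function calls a custom event that clears all instances on the SlurpBox instanced mesh component, so the replicated flag carries the collision removal to clients.
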(event EventOnDeathServer
  (params (Damage :type Float%20%28single-precision%29) (DamageTags :type Gameplay%20Tag%20Container%20Structure%20%28by%20ref%29) (Momentum :type Vector :default "0, 0, 0") (HitInfo :type Hit%20Result%20Structure%20%28by%20ref%29) (InstigatedBy :type Controller%20Object%20Reference) (DamageCauser :type Actor%20Object%20Reference) (EffectContext :type Gameplay%20Effect%20Context%20Handle%20Structure))
  ;[1 1 1] OnDeathServer
  ; Trigger OnRep function to clear Instanced Static Meshes to remove collision
  (|CallParentFunction Damage DamageTags Momentum HitInfo InstigatedBy DamageCauser EffectContext)
  (Variables|InstancedStaticMesh|SetDisableInstancedSMCollision true))

(event Custom|Remove%20Instanced%20Static%20Mesh%20Collision
  ;[1 1 1] Custom Event: Remove Instanced Static Mesh Collision
  ; Clears ISM instances to get rid of collision on destroy
  (Components|InstancedStaticMesh|ClearInstances (Variables|Default|GetSlurpBox)))

(fn OnRep_DisableInstancedSMCollision
  (CallFunction|RemoveInstancedStaticMeshCollision))

(fn ConstructionScript)
