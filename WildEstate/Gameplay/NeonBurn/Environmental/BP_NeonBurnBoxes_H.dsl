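; BP_NeonBurnBoxes_H is a Blueprint for environmental boxes in the NeonBurn gameplay area, built on a parent with an overridable instanced-mesh collision removal event.
; It overrides Remove Instanced Static Mesh Collision because it has two separate Instanced Static Mesh components.
; The override clears all instances on the SodaBoxWrapMesh component, with no parent call shown.
; Its construction script is empty.
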(event EventRemoveInstancedStaticMeshCollision
  ;[1 1 1] Remove Instanced Static Mesh Collision
  ; Override parent function as this BP has 2 different Instanced Static Mesh Components
  (Components|InstancedStaticMesh|ClearInstances (Variables|Default|GetSodaBoxWrapMesh)))

(fn ConstructionScript)
